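# Sound cue for a LimeRock demon's monster groan sound effect.
# A single wave player feeds the output directly, with a boosted volume, shared demon attenuation and the Fort_Events sound class.
# It sends a small fixed level to the distant reverb submix.
# It carries per-owner concurrency override values with a low voice cap and retrigger time, while its concurrency set holds only a null entry.
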SoundCue /LimeRock/Assets/Audio/SFX/Demon/SC_LimeRock_100_Monster_Groan
root=WavePlayer_0
nodes [{id=WavePlayer_0, properties {soundWaveAssetPtr=/LimeRock/Assets/Audio/SFX/Demon/Wav/LimeRock_100_Monster_Groan}}]
properties {
 attenuationSettings=/LimeRock/Assets/Audio/Settings/ATT_LimeRock_Demon
 maxDistance=60074.1719
 soundClassObject=&Fort_Events
 volumeMultiplier=1.5
 concurrencyOverrides {bLimitToOwner=true, maxCount=2, retriggerTime=0.25}
 concurrencySet [null]
 soundSubmixSends [{minSendLevel=0, sendLevel=0.2, soundSubmix=&Fortnite_Reverb_Distant}]
}
