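# Sound Cue for the LimeRock demon monster's writhing tentacles sound effect.
# A single Wave Player feeds the output directly, with no randomization or modulation, at reduced volume.
# It uses the shared LimeRock Demon attenuation and the Fort_Events sound class.
# It stores per-owner concurrency override values, but the override flag is left at default and the concurrency set holds only a null entry.
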SoundCue /LimeRock/Assets/Audio/SFX/Demon/SC_LimeRock_100_Monster_TentaclesWrithing
root=WavePlayer_0
nodes [{id=WavePlayer_0, properties {soundWaveAssetPtr=/LimeRock/Assets/Audio/SFX/Demon/Wav/LimeRock_100_Tentacles_Writhing}}]
properties {
 attenuationSettings=/LimeRock/Assets/Audio/Settings/ATT_LimeRock_Demon
 maxDistance=60074.1719
 soundClassObject=&Fort_Events
 volumeMultiplier=0.6
 concurrencyOverrides {bLimitToOwner=true, maxCount=2, retriggerTime=0.25}
 concurrencySet [null]
}
